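# Physics asset for the Kepler Female two-handed pickaxe skeletal mesh, providing collision and physical motion for its dangling tubes.
# Two large capsules on the handle bone approximate the weapon body, while four small capsules cover the dynamic tube bones from bottom to top.
# Each tube is constrained directly to the handle, not chained, with limited swing that varies per tube and a tight, uniform twist limit.
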PhysicsAsset /BRCosmetics/Weapons/FORT_Melee/Pickaxe_Kepler_Female/Meshes/Kepler_Female_Axe_Physics
bodies {
 Melee_TwoHanded_Handle [
  {shapeType=Capsule, length=104.1468, radius=17.5576, center=<0.8964 10.1558 25.0516>, rotation=<rot -0.5778 -89.8923 0.5978>}
  {shapeType=Capsule, length=56.8776, radius=19.9239, center=<-0.4264 29.5979 91.5307>, rotation=<rot 0 0 90.931>}
 ]
 dyn_tube_bottom_00 [{shapeType=Capsule, length=5.13, radius=1.255, center=<11.1949 0.1567 0.1924>, rotation=<rot 90 -17.354 -21.13>}]
 dyn_tube_mid_bottom_00 [{shapeType=Capsule, length=5.13, radius=1.255, center=<11.1949 0.1567 0.1924>, rotation=<rot 90 180 -167.0797>}]
 dyn_tube_mid_top_00 [{shapeType=Capsule, length=5.13, radius=1.255, center=<11.1949 0.1567 0.1924>, rotation=<rot 90 180 -167.0797>}]
 dyn_tube_top_00 [{shapeType=Capsule, length=5.13, radius=1.255, center=<14.7309 -0.095 -1.8722>, rotation=<rot 90 12.3391 9.9514>}]
}
constraints [
 {bone1Name=dyn_tube_bottom_00, bone2Name=Melee_TwoHanded_Handle, swing1LimitDegrees=20, swing2LimitDegrees=20, twistLimitDegrees=10}
 {bone1Name=dyn_tube_mid_bottom_00, bone2Name=Melee_TwoHanded_Handle, swing1LimitDegrees=25, swing2LimitDegrees=25, twistLimitDegrees=10}
 {bone1Name=dyn_tube_mid_top_00, bone2Name=Melee_TwoHanded_Handle, swing1LimitDegrees=35, swing2LimitDegrees=35, twistLimitDegrees=10}
 {bone1Name=dyn_tube_top_00, bone2Name=Melee_TwoHanded_Handle, swing1LimitDegrees=20, swing2LimitDegrees=35, twistLimitDegrees=10}
]
