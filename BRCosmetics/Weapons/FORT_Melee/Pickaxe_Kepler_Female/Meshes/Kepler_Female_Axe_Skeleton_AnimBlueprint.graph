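# Animation Blueprint for the Kepler Female pickaxe skeleton that animates its claw around the player's melee swings.
# Outside the front end, it polls PlayerMeleeCombo to blend in a claw rotate-and-pinch pose while swinging, then switches to the axe-open sequence for 1.5 seconds afterward.
# A RigidBody node adds physics to the result, weighted lower in the front end.
# A time-based spin rotator is computed but unused in the AnimGraph.
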AnimBlueprint /BRCosmetics/Weapons/FORT_Melee/Pickaxe_Kepler_Female/Meshes/Kepler_Female_Axe_Skeleton_AnimBlueprint
target_skeleton=&/Kepler_Female_Axe_Skeleton
graphs [
 AnimGraph {nodes [
   "Transform (Modify) Bone - Bone: Melee_TwoHanded_Handle"(Alpha=1, Translation="0,0,0", Rotation="0,0,0", Scale="0.9,0.9,0.9") {
    rotationMode=BMM_Replace
    rotationSpace=BCS_BoneSpace
    scaleMode=BMM_Replace
    scaleSpace=BCS_BoneSpace
    translationMode=BMM_Replace
    translationSpace=BCS_BoneSpace
    alphaBoolBlend {blendOption=Linear}
    boneToModify {boneName=Melee_TwoHanded_Handle}
    scale=<0.9 0.9 0.9>
   }
   RigidBody() {
    bClampLinearTranslationLimitToRefPose=true
    bFreezeIncomingPoseOnStart=true
    bOverrideWorldGravity=true
    overridePhysicsAsset=&/Kepler_Female_Axe_Physics
    simulationSpace=BaseBoneSpace
    alphaBoolBlend {blendOption=Linear}
    baseBoneRef {boneName=b_AR_Root}
    componentLinearAccScale=<0.4 0.4 0.4>
    componentLinearVelScale=<0.2 0.2 0.2>
    overrideWorldGravity {z=-980}
   }
   n1 = "Transform (Modify) Bone - Bone: dyn_claw_rotate"(ComponentPose=Local To Component(LocalPose=Use cached pose 'claw'()), Alpha=1, Translation="0,0,0", Rotation="0,0,180", Scale="1,1,1") {
    rotationMode=BMM_Replace
    rotationSpace=BCS_BoneSpace
    scaleMode=BMM_Replace
    scaleSpace=BCS_BoneSpace
    translationMode=BMM_Replace
    translationSpace=BCS_BoneSpace
    alphaBoolBlend {blendOption=Linear}
    boneToModify {boneName=dyn_claw_rotate}
    rotation {roll=180}
   }
   n2 = "Transform (Modify) Bone - Bone: dyn_claw_pinch_top"(ComponentPose=@n1, Alpha=1, Translation="0,0,-9", Rotation="0,0,0", Scale="1,1,1") {
    rotationMode=BMM_Replace
    rotationSpace=BCS_BoneSpace
    scaleMode=BMM_Replace
    scaleSpace=BCS_BoneSpace
    translationMode=BMM_Replace
    translationSpace=BCS_BoneSpace
    alphaBoolBlend {blendOption=Linear}
    boneToModify {boneName=dyn_claw_pinch_top}
    translation {z=-9}
   }
   n3 = "Transform (Modify) Bone - Bone: dyn_claw_pinch_bottom"(ComponentPose=@n2, Alpha=1, Translation="0,0,9", Rotation="0,0,0", Scale="1,1,1") {
    rotationMode=BMM_Replace
    rotationSpace=BCS_BoneSpace
    scaleMode=BMM_Replace
    scaleSpace=BCS_BoneSpace
    translationMode=BMM_Replace
    translationSpace=BCS_BoneSpace
    alphaBoolBlend {blendOption=Linear}
    boneToModify {boneName=dyn_claw_pinch_bottom}
    translation {z=9}
   }
   n4 = Two Way Blend(A=Use cached pose 'claw'(), B=Component To Local(ComponentPose=@n3), Alpha=Get Is_Swinging() -> (Is_Swinging:double)) {blendNode {
     bResetChildOnActivation=true
     alphaBoolBlend {blendOption=Linear}
     alphaScaleBiasClamp {bInterpResult=true, interpSpeedDecreasing=3, interpSpeedIncreasing=8}
    }}
   n5 = Two Way Blend(
    A=@n4
    B=Sequence Player 'Kepler_Female_Axe_open'() {sequence=&/Kepler_Female_Axe_open}
    Alpha=Get Stage1() -> (Stage1:double)
   ) {blendNode {bResetChildOnActivation=true, alphaBoolBlend {blendOption=Linear}, alphaScaleBiasClamp {interpSpeedDecreasing=3, interpSpeedIncreasing=8}}}
   n6 = RigidBody(ComponentPose=Local To Component(LocalPose=@n5), Alpha=SelectFloat(A=0.35, B=0.8, bPickA=Get bIsInFrontEnd())) {
    bOverrideWorldGravity=true
    overridePhysicsAsset=&Kepler_Female_Axe_Physics
    simulationSpace=BaseBoneSpace
    alphaBoolBlend {blendOption=Linear}
    baseBoneRef {boneName=b_AR_Root}
    componentLinearAccScale=<0.3 0.3 0.3>
    componentLinearVelScale=<0.2 0.2 0.2>
    overrideWorldGravity {z=-980}
   }
   Output Pose(Result=Component To Local(ComponentPose=@n6)) {layerGroup=None}
   Save cached pose 'claw'(Pose=Input Pose()) {cacheName=claw}
   Two Way Blend(Alpha=0) {blendNode {
     bResetChildOnActivation=true
     alphaBoolBlend {blendOption=Linear}
     alphaScaleBiasClamp {bInterpResult=true, interpSpeedDecreasing=3, interpSpeedIncreasing=8}
    }}
  ]}
 EventGraph {nodes [
   Event BlueprintUpdateAnimation()
   > Branch(Condition=NOT Boolean(A=Get bIsInFrontEnd()))
   > n1 = Sequence!(execute) -> (then_0, then_1, then_2:Exec, then_3:Exec, then_4:Exec)
   n2 = Montage_IsPlaying(
    self=GetOwnerFortPlayerAnimInstance!(self:Custom Character Part Anim Instance Object Reference) -> (ReturnValue:Fort Player Anim Instance Object Reference)
    Montage=/Game/Animation/Game/MainPlayer/Montages/Melee/PlayerMeleeCombo
   )
   Set Is_Swinging(execute=@n1.then_2, Is_Swinging:double="To Float (Boolean)"(InBool=@n2)) -> (Output_Get:double)
   > Set Was Swinging(Was Swinging:bool=Get swing() -> (swing:bool)) -> (Output_Get:bool)
   > Set swing(swing:bool=@n2) -> (Output_Get:bool)
   > Branch(Condition=Get swing())
   n3 = Get swing()
   Branch(execute=@n1.then_3, Condition=AND Boolean(A=Get Was Swinging() -> (Was Swinging:bool), B=NOT Boolean(A=@n3)))
   > Do Once(Reset=@n4.then)
   > Set Stage1(Stage1:double=1) -> (Output_Get:double)
   > Delay(Duration=1.5)
   > n4 = Set Stage1(Stage1=0)
   n5 = Spin FNC!(execute=@n1.then_0, self, speed:double=-50, initVal:double=0) -> (then, output:double)
   > Set Spinning_Rotator_01(Spinning_Rotator_01:Rotator=MakeRotator(Roll=@n5)) -> (Output_Get:Rotator)
   Event BlueprintInitializeAnimation()
   > Set Skel_Mesh(Skel_Mesh=GetOwnerSkeletalMeshComponent())
   > n6 = Cast To FortPlayerPawn(Object=TryGetFortPawn())
   > Set Pawn(Pawn:Fort Player Pawn Object Reference=@n6) -> (Output_Get:Fort Player Pawn Object Reference)
   TryGetPawnOwner()
   PrintString!(
    execute
    InString=GetDisplayName(Object=GetCurrentActiveMontage!(self:Anim Instance Object Reference) -> (ReturnValue:Anim Montage Object Reference))
    bPrintToScreen
    bPrintToLog
    TextColor
    Duration
    Key
   ) -> (then)
   AND Boolean(A=false, B=NOT Boolean(A=false))
   Get bIsInFrontEnd()
  ]}
 Spin FNC {nodes [
   n1 = Spin FNC!() -> (then, speed:double=~, initVal:double=~) {is=title K2Node_FunctionEntry}
   n2 = float * float(A=@n1, B=GetRealTimeSeconds())
   > Return Node!(
    execute
    output:double=NormalizeAxis!(Angle:float=float + float(A=@n1.initVal, B=@n2) {is=title K2Node_CommutativeAssociativeBinaryOperator}) -> (ReturnValue:float)
   )
  ]}
]
variables [
 Is_Swinging {type=number}
 Pawn {type=FortPlayerPawn}
 Skel_Mesh {type=SkeletalMeshComponent}
 Spinning_Rotator_01 {type=Rotator, default=<rot 0 0 169.125>}
 Stage1 {type=number}
 Stage2 {type=number}
 Stage3 {type=number}
 Stage4 {type=number}
 Stage5 {type=number}
 Stage6 {type=number}
 Stage7 {type=number}
 swing {type=boolean}
 Was Swinging {type=boolean}
]
